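# Material function for the Throne environment that animates vines through World Position Offset on incoming material attributes.
# A looping, phase-shifted time sweep plus a smoothed UV V gradient samples a curve atlas row.
# The curve value, optionally squared, scales displacement along vertex normals, with separate vertical strength apparently blended by an up-facing normal mask.
# Static switches control additive or replacing offset and bypassing the animation; a CurveTime output is also exposed.
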MaterialFunction /Hermes/Environments/Throne/Materials/Functions/MF_QTAnim
graph {nodes [
  n1 = GetMaterialAttributes(None=NamedRerouteDeclaration(None=FunctionInput() {inputName=InMA, inputType=FunctionInput_MaterialAttributes}) {name=inMA, nodeColor=<rgba 0 0.7891 1 1>}) {attributeGetTypes [#1]}
  n2 = NamedRerouteUsage()
  n3 = StaticSwitch(
   True=Add(A=@n1.WorldPositionOffset, B=@n2)
   False=@n2
   Value=FunctionInput(Preview=StaticBool() {value=true}) {bUsePreviewValueAsDefault=true, inputName=AdditiveWPO, inputType=FunctionInput_StaticBool, sortPriority=1}
  )
  FunctionOutput(None=StaticSwitch(
    True=SetMaterialAttributes!(MaterialAttributes=@n1.MaterialAttributes, WorldPositionOffset=@n3) {attributeSetTypes [#1]}
    False=NamedRerouteUsage()
    Value=NamedRerouteUsage()
   )) {outputName=outMA}
  FunctionOutput(None=NamedRerouteUsage()) {outputName=CurveTime, sortPriority=1}
  NamedRerouteDeclaration(None=FunctionInput(Preview=StaticBool()) {bUsePreviewValueAsDefault=true, inputName=UseQTAnim, inputType=FunctionInput_StaticBool, sortPriority=1}) {name=useAnim, nodeColor=<rgba 1 0.8906 0 1>}
  n4 = VertexNormalWS()
  n5 = Clamp(None=ScalarParameter() {group=Vine Animation, parameterName=VinesAnim Smooth, sliderMax=0.99}) {maxDefault=0.99}
  n6 = TextureCoordinate()
  n7 = NamedRerouteDeclaration(None=Multiply(A=OneMinus(None=@n5), B=ComponentMask(None=@n6) {g=true})) {name=UVFadeMask, nodeColor=<rgba 0 1 0.6953 1>}
  n8 = Add(
   A=ScalarParameter() {group=Vine Animation, parameterName=VinesAnim Phase}
   B=Multiply(A=Time(), B=ScalarParameter() {defaultValue=1, group=Vine Animation, parameterName=VinesAnim Speed})
  )
  n9 = Subtract(A=Multiply(A=Frac(None=@n8)) {constB=2})
  n10 = CurveAtlasRowParameter(CurveTime=Saturate(None=Add(A=@n7, B=@n9))) {
   atlas=/Hermes/Environments/Throne/Materials/Functions/Curves/T_QTCurveAtlas
   curve=/Hermes/Environments/Throne/Materials/Functions/Curves/Anim_Curve_01
   defaultValue=6
   group=Vine Animation
   parameterName=VinesAnim Curve
  }
  n11 = Multiply(A=@n10.R, B=@n10.R)
  n12 = Multiply(
   A=LinearInterpolate(
    A=ScalarParameter() {defaultValue=1, group=Vine Animation, parameterName=VinesAnim Displacement}
    B=ScalarParameter() {defaultValue=1, group=Vine Animation, parameterName=VinesAnim DisplacementZ}
    Alpha=NamedRerouteUsage()
   )
   B=NamedRerouteDeclaration(None=StaticSwitchParameter(True=@n11, False=@n10.R) {desc=Use Curve alpha as a mask, group=Vine Animation, parameterName=VinesAnim CurveAlphaMask, sortPriority=255}) {name=curveParam, nodeColor=<rgba 1 0.7031 0 1>}
  )
  NamedRerouteDeclaration(None=Multiply(A=Multiply(A=@n4, B=@n12), B=NamedRerouteUsage())) {name=outAnimCurve, nodeColor=<rgba 1 0 0.0703 1>}
  NamedRerouteDeclaration(None=DotProduct(A=Constant3Vector() {constant=<rgba 0 0 1 0>}, B=@n4)) {name=topMask, nodeColor=<rgba 1 0 0.5625 1>}
 ]}
