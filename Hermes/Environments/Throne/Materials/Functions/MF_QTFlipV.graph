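# MF_QTFlipV is a material function that flips the V coordinate of a UV set.
# It splits the InUV input into components with BreakOutFloat2Components, inverts G with OneMinus, and appends it back to the unchanged R.
# The InUV input uses the mesh texture coordinate as its default when left unconnected.
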MaterialFunction /Hermes/Environments/Throne/Materials/Functions/MF_QTFlipV
graph {nodes [
  n1 = MaterialFunctionCall(Float2=FunctionInput(Preview=TextureCoordinate()) {bUsePreviewValueAsDefault=true, inputName=InUV, inputType=FunctionInput_Vector2}) {materialFunction=&BreakOutFloat2Components}
  FunctionOutput(None=AppendVector(A=@n1.R, B=OneMinus(None=@n1.G)))
 ]}
