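# Material function that triplanar-projects a diffuse texture in world space, returning XYZ, XY and Z results plus blend alphas and the UV divisor.
# World position divided by negated texture size drives three shared-sampler projections, with optional separate Z texture and size.
# Contrast-adjusted absolute normal X and Z drive nested lerps, optionally refined by height-lerping texture alpha or a chosen channel.
# Static switches gate these options and float4 diffuse export.
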MaterialFunction /Hermes/Environments/Throne/Materials/Functions/MF_WorldAlignedTextures_Simple
graph {nodes [
  FunctionOutput(None=AppendVector(A=NamedRerouteUsage(), B=NamedRerouteUsage())) {outputName=XZLerpAlpha, sortPriority=4}
  n1 = FunctionInput(Preview=WorldPosition().XYZ) {bUsePreviewValueAsDefault=true, inputName=WorldPosition, sortPriority=200}
  n2 = FunctionInput() {
   bUsePreviewValueAsDefault=true
   description=Size of texture in worldspace units
   inputName=TextureSize
   sortPriority=-11
   previewValue=<64 64 64 6>
  }
  n3 = Abs(None=@n2)
  n4 = Divide(A=@n1, B=NamedRerouteDeclaration(None=Multiply(A=@n3) {constB=-1}) {name=NWPD, nodeColor=<rgba 1 0 0.4453 1>})
  n5 = TextureObject() {texture=/Engine/EngineMaterials/DefaultDiffuse}
  n6 = FunctionInput(Preview=@n5) {bUsePreviewValueAsDefault=true, inputName=Diffuse Texture Object, inputType=FunctionInput_Texture2D, sortPriority=-100}
  n7 = TextureSample(UVs=ComponentMask(None=@n4) {b=true, r=true}, Tex=@n6) {desc=Y projection, samplerSource=SSM_Wrap_WorldGroupSettings}
  n8 = AppendVector(A=@n7.RGB, B=@n7.RGB)
  n9 = FunctionInput(Preview=StaticBool()) {
   bUsePreviewValueAsDefault=true
   description="Default = False"
   inputName=Export Float 4 Diffuse
   inputType=FunctionInput_StaticBool
   sortPriority=105
  }
  n10 = TextureSample(UVs=ComponentMask(None=@n4) {b=true, g=true}, Tex=@n6) {desc=X projection, samplerSource=SSM_Wrap_WorldGroupSettings}
  n11 = AppendVector(A=@n10.RGB, B=@n10.RGB)
  n12 = FunctionInput(Preview=VertexNormalWS()) {bUsePreviewValueAsDefault=true, inputName=World Space Normal, sortPriority=201, previewValue=<0 0 1 1>}
  n13 = ComponentMask(None=@n12) {r=true}
  n14 = FunctionInput() {bUsePreviewValueAsDefault=true, inputName=ProjectionTransitionContrast, inputType=FunctionInput_Scalar, sortPriority=7}
  n15 = ComponentMask(None=MaterialFunctionCall(In=Abs(None=@n13), Contrast=@n14) {materialFunction=&CheapContrast}.Result) {desc="1=+x", r=true}
  n16 = FunctionInput() {
   bUsePreviewValueAsDefault=true
   description="(If pick the diffuse height channel == true)  (1,0,0,0) will use the diffuse red channel as the height map. (0,1,0,0) will use the green channel etc."
   inputName=TheDiffuseHeightChannel
   inputType=FunctionInput_Vector4
   sortPriority=9
  }
  n17 = FunctionInput(Preview=StaticBool()) {
   bUsePreviewValueAsDefault=true
   description="Default = False"
   inputName=PickTheDiffuseHeightChannel
   inputType=FunctionInput_StaticBool
   sortPriority=9
  }
  n18 = FunctionInput() {bUsePreviewValueAsDefault=true, inputName=HeightLerpTransitionContrast, inputType=FunctionInput_Scalar, sortPriority=9}
  n19 = MaterialFunctionCall!(
   Diffuse 1
   Diffuse 2
   Transistion Phase=@n15
   Height Texture 1=StaticSwitch(True=DotProduct(A=@n8, B=@n16), False=@n7.A, Value=@n17)
   Height Texture 2=StaticSwitch(True=DotProduct(A=@n11, B=@n16), False=@n10.A, Value=@n17)
   Contrast=@n18
  ) {materialFunction=/Engine/Functions/Engine_MaterialFunctions02/Texturing/HeightLerpWithTwoHeightMaps}
  n20 = OneMinus(None=@n19.Alpha)
  n21 = FunctionInput(Preview=StaticBool()) {
   bUsePreviewValueAsDefault=true
   description=Defaults to false.
   inputName=Use Texture Alpha Heightmap to Refine the Transistion
   inputType=FunctionInput_StaticBool
   sortPriority=8
   previewValue=<1 0 0 1>
  }
  n22 = LinearInterpolate(
   A=StaticSwitch(True=@n8, False=@n7.RGB, Value=@n9)
   B=StaticSwitch(True=@n11, False=@n10.RGB, Value=@n9)
   Alpha=NamedRerouteDeclaration(None=StaticSwitch(True=@n20, False=@n15, Value=@n21)) {name=XLerpAlpha, nodeColor=<rgba 0.7578 0 1 1>}
  ) {desc="Float3 1 = x projection"}
  n23 = FunctionInput() {
   bUsePreviewValueAsDefault=true
   description=Size of texture in worldspace units
   inputName=Texture Size Z projection
   previewValue=<64 64 64 6>
  }
  n24 = Multiply(A=@n23) {constB=-1}
  n25 = FunctionInput(Preview=StaticBool()) {
   bUsePreviewValueAsDefault=true
   inputName=Use seperate settings for Z projection
   inputType=FunctionInput_StaticBool
   previewValue=<64 64 64 6>
  }
  n26 = ComponentMask(None=StaticSwitch(True=Divide(A=@n1, B=@n24), False=@n4, Value=@n25)) {g=true, r=true}
  n27 = FunctionInput(Preview=@n5) {bUsePreviewValueAsDefault=true, inputName=Z Plane Diffuse Texture Object, inputType=FunctionInput_Texture2D, sortPriority=4}
  n28 = TextureSample(UVs=@n26, Tex=@n27) {samplerSource=SSM_Wrap_WorldGroupSettings}
  n29 = TextureSample(UVs=@n26, Tex=@n6) {samplerSource=SSM_Wrap_WorldGroupSettings}
  n30 = FunctionInput(Preview=StaticBool()) {
   bUsePreviewValueAsDefault=true
   inputName=Use seperate texture for Z projection
   inputType=FunctionInput_StaticBool
   previewValue=<64 64 64 6>
  }
  n31 = StaticSwitch(True=AppendVector(A=@n28.RGB, B=@n28.RGB), False=AppendVector(A=@n29.RGB, B=@n29.RGB), Value=@n30) {desc=Z projection}
  n32 = StaticSwitch(True=@n31, False=ComponentMask(None=@n31) {b=true, g=true, r=true}, Value=@n9)
  n33 = ComponentMask(None=@n12) {b=true}
  n34 = ComponentMask(None=MaterialFunctionCall(In=Abs(None=@n33), Contrast=@n14) {materialFunction=&CheapContrast}.Result) {desc="1=+z", r=true}
  n35 = StaticSwitch(True=DotProduct(A=@n31, B=@n16), False=ComponentMask(None=@n31) {a=true}, Value=@n17)
  n36 = StaticSwitch(
   True=OneMinus(None=MaterialFunctionCall(Transistion Phase=@n34, Height Texture 1=@n19.Resulting Height Map, Height Texture 2=@n35, Contrast=@n18) {materialFunction=&HeightLerpWithTwoHeightMaps}.Alpha)
   False=@n34
   Value=@n21
  )
  FunctionOutput(None=LinearInterpolate(A=@n22, B=@n32, Alpha=NamedRerouteDeclaration(None=@n36) {name=ZLerpAlpha, nodeColor=<rgba 0 1 0.4375 1>}) {desc="1 = z projection"}) {outputName=XYZ Texture, sortPriority=1}
  FunctionOutput(None=@n22) {outputName=XY Texture, sortPriority=2}
  FunctionOutput(None=@n32) {outputName=Z Texture, sortPriority=3}
  FunctionOutput(None=NamedRerouteUsage()) {outputName=NegativeWDivisor, sortPriority=5}
 ]}
