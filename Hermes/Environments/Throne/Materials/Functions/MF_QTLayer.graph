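# MF_QTLayer is a Throne environment material function that builds one surface layer as Material Attributes from albedo, normal and packed AOSR textures.
# Inputs provide UV and world-aligned triplanar sampling with per-map blend alphas, plus optional detail maps and animated-UV AOSR.
# Global channel-mask parameters unpack AO, specular, roughness and emissive, with AOER/SER switches for alternate packing.
# A static switch can output attributes without base color and normal, with metallic held constant.
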MaterialFunction /Hermes/Environments/Throne/Materials/Functions/MF_QTLayer
graph {nodes [
  Constant3Vector() {constant=<rgba 0 0 1 0>}
  FunctionInput(Preview=AppendVector(A=TextureCoordinate(), B=Constant())) {bUsePreviewValueAsDefault=true, desc="U Tile, V Tile, UV Index", inputName=UVDetailTileIndex, sortPriority=8}
  n1 = Constant()
  n2 = SetMaterialAttributes!(
   MaterialAttributes
   Specular=NamedRerouteUsage()
   Roughness=NamedRerouteUsage()
   AmbientOcclusion=NamedRerouteUsage()
   Metallic=@n1
   EmissiveColor=NamedRerouteUsage()
  ) {attributeSetTypes [
    #1
    #2
    #3
    #4
    #5
   ]}
  n3 = NamedRerouteUsage()
  n4 = NamedRerouteUsage()
  n5 = Add(A=@n3, B=@n4)
  n6 = StaticSwitch(
   True=LinearInterpolate(A=NamedRerouteUsage(), B=Multiply(A=@n5) {constB=0.5}, Alpha=NamedRerouteUsage())
   False=NamedRerouteUsage()
   Value=NamedRerouteUsage()
  )
  n7 = ChannelMaskParameter(None=@n6) {group=B_GlobalMaskChannels, maskChannel=Green, parameterName=Emissive Channel, defaultValue=<rgba 0 1 0 0>}
  n8 = Multiply(A=@n7.R, B=NamedRerouteUsage())
  n9 = StaticSwitch(True=@n8, False=NamedRerouteUsage(), Value=NamedRerouteUsage())
  n10 = NamedRerouteUsage()
  n11 = SetMaterialAttributes!(
   MaterialAttributes
   BaseColor=NamedRerouteUsage()
   Specular=NamedRerouteUsage()
   Roughness=NamedRerouteUsage()
   AmbientOcclusion=NamedRerouteUsage()
   Normal=NamedRerouteUsage()
   Metallic=@n1
   EmissiveColor=NamedRerouteDeclaration(None=StaticSwitch(True=@n8, False=@n9, Value=@n10)) {name=EmOut, nodeColor=<rgba 0 0.2734 1 1>}
  ) {attributeSetTypes [
    #6
    #1
    #2
    #3
    #7
    #4
    #5
   ]}
  FunctionOutput(None=StaticSwitch(True=@n2, False=@n11, Value=NamedRerouteUsage()))
  Multiply(A=@n3, B=@n4)
  NamedRerouteDeclaration(None=FunctionInput(Preview=StaticBool()) {bUsePreviewValueAsDefault=true, inputName=UseUVDetailMaps, inputType=FunctionInput_StaticBool, sortPriority=7}) {name=useUVDetail, nodeColor=<rgba 0.0312 0 1 1>}
  n12 = NamedRerouteDeclaration(None=FunctionInput() {inputName=UVs, inputType=FunctionInput_Vector2}) {name=UVs, nodeColor=<rgba 0.9922 0 1 1>}
  n13 = TextureObject()
  n14 = FunctionInput(Preview=@n13) {inputName=Tobj_Albedo, inputType=FunctionInput_Texture2D, sortPriority=1}
  NamedRerouteDeclaration(None=TextureSample(UVs=@n12, Tex=NamedRerouteDeclaration(None=@n14) {name=albedoOBJ, nodeColor=<rgba 1 0.75 0 1>}).RGB) {name=albedoUV, nodeColor=<rgba 1 0 0.2109 1>}
  n15 = TextureSample(
   UVs=@n12
   Tex=NamedRerouteDeclaration(None=FunctionInput(Preview=@n13) {inputName=Tobj_AOSR, inputType=FunctionInput_Texture2D, sortPriority=3}) {name=AOSROBJ, nodeColor=<rgba 0.4297 0 1 1>}
  )
  NamedRerouteDeclaration(None=StaticSwitch(
    True=NamedRerouteUsage()
    False=NamedRerouteDeclaration(None=@n15.RGB) {name=staticAOSR, nodeColor=<rgba 0.2188 0 1 1>}
    Value=NamedRerouteUsage()
   )) {name=AOSRUV, nodeColor=<rgba 0 0.5312 1 1>}
  n16 = FunctionInput(Preview=TextureObject() {samplerType=SAMPLERTYPE_Normal, texture=/Engine/EngineMaterials/FlatNormal}) {inputName=Tobj_Normal, inputType=FunctionInput_Texture2D, sortPriority=2}
  NamedRerouteDeclaration(None=TextureSample(UVs=@n12, Tex=NamedRerouteDeclaration(None=@n16) {name=normalOBJ, nodeColor=<rgba 0 1 0.1094 1>}).RGB) {name=normalUV, nodeColor=<rgba 0.8516 1 0 1>}
  NamedRerouteDeclaration(None=MaterialFunctionCall!(
    Diffuse Texture Object=NamedRerouteUsage()
    TextureSize=NamedRerouteUsage()
    Texture Size Z projection
    Use seperate settings for Z projection
    Use seperate texture for Z projection
    Z Plane Diffuse Texture Object
    ProjectionTransitionContrast
    Use Texture Alpha Heightmap to Refine the Transistion
    HeightLerpTransitionContrast
    PickTheDiffuseHeightChannel
    TheDiffuseHeightChannel
    Export Float 4 Diffuse
    WorldPosition
    World Space Normal
   ) {materialFunction=&/MF_WorldAlignedTextures_Simple}.XYZ Texture) {name=albedoTri, nodeColor=<rgba 0.5 0 1 1>}
  n17 = MaterialFunctionCall(Diffuse Texture Object=NamedRerouteUsage(), TextureSize=NamedRerouteUsage()) {materialFunction=&MF_WorldAlignedTextures_Simple}
  NamedRerouteDeclaration(None=@n17.XYZ Texture) {name=AOSRTri, nodeColor=<rgba 0 1 0.1797 1>}
  NamedRerouteDeclaration(None=@n17.NegativeWDivisor) {name=NWD, nodeColor=<rgba 1 0 0.1406 1>}
  NamedRerouteDeclaration(None=@n17.XZLerpAlpha) {name=XZLerp, nodeColor=<rgba 0 0.4609 1 1>}
  NamedRerouteDeclaration(None=MaterialFunctionCall!(
    TextureSize=NamedRerouteUsage()
    XZLerp=NamedRerouteUsage()
    NegativeDivisor=NamedRerouteUsage()
    UseInputXZLerp=NamedRerouteUsage()
    ProjectionTransitionContrast
    NormalTextureObject=NamedRerouteUsage()
    World Space Normals=StaticBool()
    WorldPosition
    World Space Normal
   ) {materialFunction=&/MF_WorldAlignedNormals_Simple}.Normal XYZ Texture) {name=normalTri, nodeColor=<rgba 1 0.6563 0 1>}
  NamedRerouteDeclaration(None=FunctionInput(Preview=TextureObject() {samplerType=SAMPLERTYPE_Normal, texture=&FlatNormal}) {inputName=Tobj_DetailNormal, inputType=FunctionInput_Texture2D, sortPriority=5}) {name=detailNormalOBJ, nodeColor=<rgba 1 0 0.6094 1>}
  NamedRerouteDeclaration(None=FunctionInput(Preview=StaticBool()) {bUsePreviewValueAsDefault=true, inputName=AllowTriplanar, inputType=FunctionInput_StaticBool, sortPriority=4}) {name=allowTriplanar, nodeColor=<rgba 0 1 0.6719 1>}
  NamedRerouteDeclaration(None=FunctionInput(Preview=StaticBool()) {bUsePreviewValueAsDefault=true, inputName=UseAOSRDetail, inputType=FunctionInput_StaticBool, sortPriority=6}) {name=useAOSRDetail, nodeColor=<rgba 0 0.9297 1 1>}
  NamedRerouteDeclaration(None=FunctionInput(Preview=TextureObject()) {inputName=Tobj_DetailAOSR, inputType=FunctionInput_Texture2D, sortPriority=6}) {name=detailAOSROBJ, nodeColor=<rgba 1 0.75 0 1>}
  NamedRerouteDeclaration(None=FunctionInput(Preview=StaticBool()) {bUsePreviewValueAsDefault=true, inputName=UseAOER, inputType=FunctionInput_StaticBool, sortPriority=9}) {name=useAOER, nodeColor=<rgba 0.0312 0 1 1>}
  NamedRerouteDeclaration(None=FunctionInput() {bUsePreviewValueAsDefault=true, inputName=Emissive Color, sortPriority=10}) {name=emColor, nodeColor=<rgba 0.875 0 1 1>}
  NamedRerouteDeclaration(None=FunctionInput(Preview=StaticBool()) {bUsePreviewValueAsDefault=true, inputName=UseMasksOnly, inputType=FunctionInput_StaticBool, sortPriority=11}) {name=useMasksOnly, nodeColor=<rgba 0.0312 0 1 1>}
  n18 = ChannelMaskParameter(None=@n6) {group=B_GlobalMaskChannels, maskChannel=Blue, parameterName=Roughness Channel, defaultValue=<rgba 0 0 1 0>}
  NamedRerouteDeclaration(None=@n18.R) {name=ROut, nodeColor=<rgba 0.9922 1 0 1>}
  n19 = NamedRerouteUsage()
  n20 = NamedRerouteUsage()
  n21 = LinearInterpolate(
   A=@n19
   B=MaterialFunctionCall(BaseNormal=@n19, AdditionalNormal=NamedRerouteUsage()) {materialFunction=&BlendAngleCorrectedNormals}.Result
   Alpha=@n20
  )
  n22 = NamedRerouteUsage()
  n23 = NamedRerouteUsage()
  NamedRerouteDeclaration(None=Normalize(VectorInput=StaticSwitch(True=@n21, False=@n22, Value=@n23))) {name=NOut, nodeColor=<rgba 0.6875 0 1 1>}
  n24 = ChannelMaskParameter(None=@n6) {group=B_GlobalMaskChannels, maskChannel=Green, parameterName=Specular Channel, defaultValue=<rgba 0 1 0 0>}
  NamedRerouteDeclaration(None=StaticSwitch(True=NamedRerouteUsage(), False=@n24.R, Value=NamedRerouteUsage())) {name=SOut, nodeColor=<rgba 0 1 0.3672 1>}
  NamedRerouteDeclaration(None=StaticSwitch(
    True=Constant() {r=1}
    False=ChannelMaskParameter(None=@n6) {group=B_GlobalMaskChannels, parameterName=AO Channel}.R
    Value=NamedRerouteUsage()
   )) {name=AOOut, nodeColor=<rgba 1 0.0469 0 1>}
  n25 = FunctionInput() {
   bUsePreviewValueAsDefault=true
   description="Triplanar for Albedo, Normal, AOSR"
   inputName=TriplanarSize
   sortPriority=4
   previewValue=<64 64 64 1>
  }
  NamedRerouteDeclaration(None=@n25) {name=triplanarSize, nodeColor=<rgba 1 0.3516 0 1>}
  NamedRerouteDeclaration(None=FunctionInput(Preview=StaticBool()) {bUsePreviewValueAsDefault=true, inputName=UseUVAnimation, inputType=FunctionInput_StaticBool, sortPriority=12}) {name=useUVAnim, nodeColor=<rgba 0.7145 1 0.4886 1>}
  n26 = FunctionInput() {bUsePreviewValueAsDefault=true, inputName=SpecularValue, inputType=FunctionInput_Scalar, sortPriority=10, previewValue=<0.5 0 0 1>}
  NamedRerouteDeclaration(None=@n26) {name=inSpecular, nodeColor=<rgba 0.6172 1 0 1>}
  NamedRerouteDeclaration(None=FunctionInput() {bUsePreviewValueAsDefault=true, inputName=AnimAOSRweights, sortPriority=13}) {name=AOSRUVAnimW, nodeColor=<rgba 1 0 0.6328 1>}
  NamedRerouteDeclaration(None=FunctionInput() {bUsePreviewValueAsDefault=true, inputName=AnimatedUVs, inputType=FunctionInput_Vector2, sortPriority=14}) {name=animatedUVs, nodeColor=<rgba 0.4766 0 1 1>}
  n27 = NamedRerouteUsage()
  n28 = NamedRerouteUsage()
  n29 = NamedRerouteUsage()
  NamedRerouteDeclaration(None=LinearInterpolate(A=NamedRerouteUsage(), B=TextureSample(UVs=@n27, Tex=@n28).RGB, Alpha=Ceil(None=@n29))) {name=animatedAOSR, nodeColor=<rgba 1 0.5391 0 1>}
  NamedRerouteDeclaration(None=FunctionInput(Preview=StaticBool()) {bUsePreviewValueAsDefault=true, inputName=UseSER, inputType=FunctionInput_StaticBool, sortPriority=9}) {name=useSER, nodeColor=<rgba 1 0.4405 0.666 1>}
  n30 = FunctionInput() {
   bUsePreviewValueAsDefault=true
   description="Triplanar for Albedo, Normal, AOSR"
   inputName=TriplanarAlpha
   inputType=FunctionInput_Vector4
   sortPriority=4
   previewValue=<1 1 1 1>
  }
  n31 = MaterialFunctionCall(Float4=NamedRerouteDeclaration(None=@n30) {name=triplanarAlpha, nodeColor=<rgba 0.0312 0 1 1>}) {materialFunction=&BreakOutFloat4Components}
  NamedRerouteDeclaration(None=@n31.R) {name=albedoAlpha, nodeColor=<rgba 0.2891 1 0 1>}
  NamedRerouteDeclaration(None=@n31.B) {name=normalAlpha, nodeColor=<rgba 1 0 0.6094 1>}
  NamedRerouteDeclaration(None=@n31.G) {name=AOSRAlpha, nodeColor=<rgba 0 0.9297 1 1>}
  Saturate(None=@n20)
  StaticSwitch(
   True=LinearInterpolate(A=NamedRerouteUsage(), B=NamedRerouteUsage(), Alpha=NamedRerouteUsage())
   False=NamedRerouteUsage()
   Value=NamedRerouteUsage()
  )
 ]}
